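# Exterior ambience layer cue for the maze at the Grand Glacier Hotel POI.
# A Mixer blends two looping waves reused from Artemis POIs, a Cornmaze "Dizzy" loop and a HauntedHouse exterior loop, with two other inputs left empty.
# The cue plays at reduced volume with shared Athena ambient attenuation and uses the indoor ambient sound class despite its exterior name.
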SoundCue /Game/Sounds/Helios/POI/GrandGlacierHotel/Ambience_GrandGlacierHotel_Maze_Layer_Ext_Cue
root=Mixer_1
nodes [
 {id=Mixer_1, children [null,null,WavePlayer_6,WavePlayer_0], properties {inputVolume [0.6,1,0.5,0.6]}}
 {
  id=WavePlayer_0
  properties {bLooping=true, soundWaveAssetPtr=/Game/Sounds/Artemis/POI/HauntedHouse/HauntedHouse_AmbientLoop_Exterior_Main_02}
 }
 {id=WavePlayer_6, properties {bLooping=true, soundWaveAssetPtr=/Game/Sounds/Artemis/POI/Cornmaze/Cornmaze_AmbientLoop_Dizzy_01}}
]
properties {
 attenuationSettings=&Fort_Ambient_Athena_Att
 maxDistance=2097152
 soundClassObject=&Fort_Ambient_Indoors_Athena
 volumeMultiplier=0.376
 preEffectBusSends [{}]
}
